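# Masked, unlit FX material for Niagara mesh particles, built for a voxel strand mesh.
# Emissive blends fake-shaded smoke brightness into particle color along UV height, using dynamic transition parameters and per-element UV1 offsets.
# Pivot Painter supplies per-voxel pivots and random values, mixed with particle random for scale offsets toward each pivot.
# Dynamic parameters expose expansion, scale and rotation with random variation, and vertex color encodes the X-axis vector.
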Material /DealershipSharedLibrary/Assets/Sets/Voxel/Materials/M_Strand_Voxel_PivotPainter
graph {
 properties {
  bAutomaticallySetUsageInEditor=false
  bGenerateSphericalParticleNormals=true
  bHasPixelAnimation=true
  bUsedWithNiagaraMeshParticles=true
  blendMode=BLEND_Masked
  previewMesh=/DealershipSharedLibrary/Assets/Sets/Voxel/Meshes/SM_Voxel_Strand
  shadingModel=MSM_Unlit
  usedShadingModels=Unlit
 }
 nodes [
  n1 = Constant3Vector() {constant=<rgba 0 0 1 0>}
  n2 = Transform(None=@n1)
  n3 = Clamp(None=Add(A=DotProduct(A=@n2, B=@n1)) {constB=1.2})
  n4 = DynamicParameter() {parameterIndex=3, defaultValue=<rgba 1.4 1 1.0154 0.25>, paramNames [BeginTransition, EndTransition, SmokeBrightness, EndRand]}
  n5 = ComponentMask(None=TextureCoordinate() {coordinateIndex=1}) {r=true}
  n6 = Multiply(A=Constant() {r=0.25}, B=@n5)
  n7 = ComponentMask(None=TextureCoordinate()) {g=true}
  n8 = Subtract(A=@n7, B=LinearInterpolate(Alpha=@n4.BeginTransition) {constA=-1})
  n9 = OneMinus(None=Add(A=Add(A=@n6, B=@n8), B=@n4.EndRand))
  MaterialFunctionCall(Emissive=LinearInterpolate(A=Multiply(A=@n3, B=@n4.SmokeBrightness), B=ParticleColor().RGB, Alpha=Clamp(None=@n9))) {materialFunction=&MF_FX_EmissiveAdjustments}
  n10 = Multiply(A=ScalarParameter() {defaultValue=1}, B=@n5) {constB=32}
  n11 = Subtract(A=@n7, B=@n4.BeginTransition)
  Multiply(A=Subtract(A=Add(A=@n4.EndTransition, B=@n10), B=@n7), B=Clamp(None=Ceil(None=@n11))) {constB=45}
  n12 = Subtract(A=NamedRerouteUsage(), B=WorldPosition().XYZ)
  Multiply(A=Clamp(None=Add()), B=@n12)
  n13 = Add(A=NamedRerouteUsage(), B=ParticleRandom())
  NamedRerouteDeclaration(None=ConstantBiasScale(None=Sine(None=@n13))) {name=Final Random, nodeColor=<rgba 1 0.0156 0.6498 1>}
  n14 = MaterialFunctionCall() {materialFunction=/Engine/Functions/Engine_MaterialFunctions02/PivotPainter/PivotPainter_PerObjectData}
  NamedRerouteDeclaration(None=@n14.Pivot Position) {name=PivotPosition, nodeColor=<rgba 1 0.0156 0.6498 1>}
  n15 = NamedRerouteUsage()
  n16 = NamedRerouteUsage()
  n17 = Multiply(A=NamedRerouteUsage(), B=OneMinus(None=Multiply(A=@n15, B=@n16)))
  NamedRerouteDeclaration(None=Multiply(A=@n12, B=OneMinus(None=@n17))) {name=WPO Scale, nodeColor=<rgba 0.1192 0.1512 1 1>}
  n18 = DynamicParameter() {paramNames [Expansion, Expansion Random, Scale, Scale Random]}
  NamedRerouteDeclaration(None=@n18.Scale) {name=DP Scale, nodeColor=<rgba 0.4459 0.0012 1 1>}
  NamedRerouteDeclaration(None=@n18.Scale Random) {name=DP Scale Random, nodeColor=<rgba 0.4459 0.0012 1 1>}
  NamedRerouteDeclaration(None=@n14.Random Value Per Element) {name=Random Value Per Element, nodeColor=<rgba 1 0.0156 0.6498 1>}
  n19 = Normalize(VectorInput=ConstantBiasScale(None=VertexColor()) {bias=-0.5, scale=2})
  NamedRerouteDeclaration(None=Normalize(VectorInput=Transform(None=@n19) {transformSourceType=TRANSFORMSOURCE_Instance})) {name=X-Axis Vector, nodeColor=<rgba 1 0.0156 0.6498 1>}
  n20 = DynamicParameter() {parameterIndex=1, defaultValue=<rgba 0 0 1 1>, paramNames [Rotation, Rotation Random, Param3, Param4]}
  NamedRerouteDeclaration(None=@n20.Rotation) {name=DP Rotation, nodeColor=<rgba 0.4459 0.0012 1 1>}
  NamedRerouteDeclaration(None=@n20.Rotation Random) {name=DP Rotation Random, nodeColor=<rgba 0.4459 0.0012 1 1>}
  NamedRerouteDeclaration(None=@n18.Expansion) {name=DP Expansion, nodeColor=<rgba 0.4459 0.0012 1 1>}
  NamedRerouteDeclaration(None=@n18.Expansion Random) {name=DP Expansion Random, nodeColor=<rgba 0.4459 0.0012 1 1>}
 ]
}
