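# Unlit Pivot Painter material for Niagara mesh particles that runs a distance-driven voxel effect with a glowing band.
# Per-voxel pivots in UV1 and UV2 give a distance from the origin, compared against particle alpha with per-voxel offsets from UV2.
# The resulting masks scale an offset toward each pivot, collapsing voxels, and blend fake top-lit shading into particle-colored glow.
# Dynamic parameters and a day/night bloom function control the glow.
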Material /DealershipSharedLibrary/Assets/Sets/Voxel/Materials/M_Voxel_PivotPainter
graph {
 properties {
  bAutomaticallySetUsageInEditor=false
  bGenerateSphericalParticleNormals=true
  bHasPixelAnimation=true
  bUsedWithNiagaraMeshParticles=true
  shadingModel=MSM_Unlit
  usedShadingModels=Unlit
 }
 nodes [
  n1 = Constant3Vector() {constant=<rgba 0 0 1 0>}
  n2 = Transform(None=@n1)
  n3 = Clamp(None=Add(A=DotProduct(A=@n2, B=@n1)) {constB=1.2})
  n4 = NamedRerouteUsage()
  n5 = Constant() {r=1}
  n6 = NamedRerouteUsage()
  n7 = TextureCoordinate() {coordinateIndex=2}
  n8 = Distance(A=AppendVector(A=TextureCoordinate() {coordinateIndex=1}, B=ComponentMask(None=@n7) {r=true}), B=Constant3Vector())
  n9 = ParticleColor()
  n10 = ComponentMask(None=TextureCoordinate() {coordinateIndex=2}) {g=true}
  n11 = Subtract(A=Add(A=@n9.A, B=Multiply(A=@n10) {constB=32}), B=NamedRerouteUsage())
  n12 = Add(A=OneMinus(None=Subtract(A=@n8, B=@n11)), B=ScalarParameter() {defaultValue=-100, parameterName=64})
  n13 = Add(A=@n12, B=Multiply(A=@n10) {constB=45})
  n14 = Clamp(None=Multiply(A=NamedRerouteUsage(), B=Subtract(A=@n6, B=@n13)))
  n15 = Multiply(A=Multiply(A=@n14, B=@n9.RGB), B=DynamicParameter() {paramNames [BloomGlow, BloomExponentIn, Param3, Param4]}.BloomGlow)
  MaterialFunctionCall(Emissive=LinearInterpolate(
    A=Multiply(A=@n3, B=@n4)
    B=MaterialFunctionCall!(DayBloom_Mult=@n5, In=@n15) {materialFunction=/Game/Packages/Fortnite_UniqueMaterials_1/MH_Materials/DayNight_Auto_Bloom_Control}.Result
    Alpha=@n14
   )) {materialFunction=&MF_FX_EmissiveAdjustments}
  n16 = Clamp(None=@n12)
  n17 = Clamp(None=Subtract(A=@n8, B=Add(A=@n9.A, B=@n10)))
  n18 = TextureCoordinate() {coordinateIndex=1}
  n19 = ComponentMask(None=TextureCoordinate() {coordinateIndex=2}) {r=true}
  n20 = TransformPosition(None=AppendVector(A=AppendVector(A=@n18, B=@n19), B=Constant() {r=1})) {transformSourceType=TRANSFORMPOSSOURCE_Instance, transformType=TRANSFORMPOSSOURCE_World}
  Multiply(A=Clamp(None=Add(A=@n16, B=@n17)), B=Subtract(A=ComponentMask(None=@n20) {b=true, g=true, r=true}, B=WorldPosition().XYZ))
  n21 = MaterialFunctionCall() {materialFunction=/Engine/Functions/Engine_MaterialFunctions02/PivotPainter/PivotPainter_PerObjectData}
  NamedRerouteDeclaration(None=@n21.Pivot Position) {name=PivotPosition, nodeColor=<rgba 1 0.0156 0.6498 1>}
  n22 = DynamicParameter() {parameterIndex=3, defaultValue=<rgba 0.384 14.1287 0.05 1>, paramNames [FireSoft, FireAmount, SmokeDarkness, RemoveScale]}
  NamedRerouteDeclaration(None=@n22.FireSoft) {name=FireSoft, nodeColor=<rgba 0.606 0.5324 1 1>}
  NamedRerouteDeclaration(None=@n21.Random Value Per Element) {name=Random Value Per Element, nodeColor=<rgba 1 0.0156 0.6498 1>}
  n23 = Normalize(VectorInput=ConstantBiasScale(None=VertexColor()) {bias=-0.5, scale=2})
  NamedRerouteDeclaration(None=Normalize(VectorInput=Transform(None=@n23) {transformSourceType=TRANSFORMSOURCE_Instance})) {name=X-Axis Vector, nodeColor=<rgba 1 0.0156 0.6498 1>}
  NamedRerouteDeclaration(None=@n22.RemoveScale) {name=RemoveScale, nodeColor=<rgba 0.606 0.5324 1 1>}
  NamedRerouteDeclaration(None=@n22.SmokeDarkness) {name=SmokeDarkness, nodeColor=<rgba 0.606 0.5324 1 1>}
  NamedRerouteDeclaration(None=@n22.FireAmount) {name=FireAmount, nodeColor=<rgba 0.606 0.5324 1 1>}
  NamedRerouteUsage()
 ]
}
